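# Sound cue for the music of the first Monster Hunter quest in Battle Pass Season 18.
# A mixer layers a reused reward-screen chest loot upgrade sound, slightly lowered in volume, with the quest music wave, which starts after a fixed short delay.
# It plays under the quests sound class with narrative quest attenuation, and a concurrency override allows one instance per owner, blocking new plays while one is active.
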SoundCue /BattlePassS18/Audio/MonsterHunter/Quest1/S18_Quest_MonsterHunter_01_Music_Cue
root=Mixer_1
nodes [
 {id=Delay_1, children [WavePlayer_0], properties {delayMax=1, delayMin=1}}
 {id=Mixer_1, children [Delay_1,WavePlayer_2], properties {inputVolume [1,0.7]}}
 {id=WavePlayer_0, properties {soundWaveAssetPtr=&/S18_Quest_MonsterHunter_01_Music}}
 {
  id=WavePlayer_2
  properties {soundWaveAssetPtr=/FNGameplayCues/Athena/Sounds/Quests/Awakenings/HightowerGrape/Fort_RewardScreen_TeamScore_ChestLootUpgrade}
 }
]
properties {
 attenuationSettings=/BattlePassS18/Audio/S18_Narrative_Quest_MonsterHunter_Att
 bOverrideConcurrency=true
 maxDistance=3512
 soundClassObject=&Fort_Quests
 volumeMultiplier=1
 concurrencyOverrides {bLimitToOwner=true, maxCount=1, resolutionRule=PreventNew}
}
