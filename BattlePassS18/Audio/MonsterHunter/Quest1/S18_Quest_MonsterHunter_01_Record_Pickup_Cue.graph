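# Sound cue for picking up the record in the Season 18 Monster Hunter quest 1.
# A Mixer layers the dedicated pickup wave at full level with a reused reward-screen chest loot upgrade sound at half level.
# An Enveloper fades that second layer in from near silence over roughly half a second.
# The cue uses the shared narrative quest attenuation and the quests sound class.
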SoundCue /BattlePassS18/Audio/MonsterHunter/Quest1/S18_Quest_MonsterHunter_01_Record_Pickup_Cue
root=Mixer_0
nodes [
 {
  id=Enveloper_0
  children [WavePlayer_2]
  properties {pitchCurve {editorCurveData {keys [{time=0}]}}, volumeCurve {editorCurveData {keys [<at 0 0.005>,<at 0 0.0006>,<at 0.4999 1.0278>]}}}
 }
 {id=Mixer_0, children [WavePlayer_0,Enveloper_0], properties {inputVolume [1,0.5]}}
 {id=WavePlayer_0, properties {soundWaveAssetPtr=&/S18_Quest_MonsterHunter_01_Record_Pickup}}
 {
  id=WavePlayer_2
  properties {soundWaveAssetPtr=/FNGameplayCues/Athena/Sounds/Quests/Awakenings/HightowerGrape/Fort_RewardScreen_TeamScore_ChestLootUpgrade}
 }
]
properties {
 attenuationSettings=/BattlePassS18/Audio/S18_Narrative_Quest_Att
 maxDistance=3512
 soundClassObject=&Fort_Quests
 volumeMultiplier=1
}
